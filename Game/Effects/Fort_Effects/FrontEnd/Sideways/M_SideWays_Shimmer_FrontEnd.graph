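# Translucent unlit Niagara sprite material for a shimmering, inky-smoke effect in the FrontEnd Sideways effects.
# A particle-random 2x2 flipbook frame, distorted by panning chromatic smoke and sphere-masked, forms the shape.
# Its square blends with distorted, panning detail smoke, tinted by particle color for emissive.
# The unsquared shape times particle alpha gives opacity, depth-faded only on the default shading path; both distortions use only their X parameters for both axes.
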Material /Game/Effects/Fort_Effects/FrontEnd/Sideways/M_SideWays_Shimmer_FrontEnd
graph {
 properties {bUsedWithNiagaraSprites=true, blendMode=BLEND_Translucent, pixelDepthOffsetMode=PDOM_Legacy, shadingModel=MSM_Unlit, usedShadingModels=Unlit}
 nodes [
  Constant2Vector() {g=-0.1, r=-0.02}
  n1 = ParticleColor()
  n2 = DynamicParameter() {defaultValue=<rgba 0 1 1 1>, paramNames [Random,Random2,Flicker,Param4]}
  n3 = Constant() {r=2}
  n4 = MaterialFunctionCall("Animation  Phase (0-1)"=@n2.Random, Number of Rows=@n3, Number of Columns=@n3) {materialFunction=&FlipBook}
  n5 = Add(A=TextureCoordinate() {uTiling=1.2, vTiling=1.2}, B=@n2.Random2)
  n6 = Multiply(A=Time(), B=ScalarParameter() {defaultValue=1, parameterName=PanSpeed})
  n7 = ComponentMask(None=TextureSample(UVs=Panner(Coordinate=@n5, Time=@n6) {speedX=-0.02, speedY=0.05}) {texture=&T_Inky_Smoke_Tile_Chromatic}.RGB) {g=true, r=true}
  n8 = DynamicParameter() {parameterIndex=1, defaultValue=<rgba 0.05 0.025 1 1>, paramNames [DistortX, DistortY, DistortX_Detail, DistortY_Detail]}
  n9 = Multiply(A=@n7, B=AppendVector(A=@n8.DistortX, B=@n8.DistortX))
  n10 = TextureCoordinate()
  n11 = Constant() {r=0.5}
  n12 = Multiply(
   A=TextureSample(UVs=Add(A=@n4.UVs, B=@n9)) {texture=/Game/Effects/Fort_Effects/Textures/T_InkySmoke_Flipbook}.G
   B=SphereMask(A=Add(A=@n9, B=@n10), B=@n11, Radius=@n11) {hardnessPercent=0}
  )
  n13 = Multiply(A=@n12, B=@n12)
  n14 = Add(A=@n2.Random, B=TextureCoordinate() {uTiling=1.5, vTiling=1.5})
  n15 = DynamicParameter() {parameterIndex=1, defaultValue=<rgba 0.05 0.025 1 1>, paramNames [DistortX, DistortY, DistortX_Detail, DistortY_Detail]}
  n16 = Multiply(A=@n7, B=AppendVector(A=@n15.DistortX_Detail, B=@n15.DistortX_Detail))
  n17 = TextureSample(UVs=Panner(Coordinate=Add(A=@n14, B=@n16), Time=@n6) {speedX=0.05, speedY=0.12}) {texture=&T_Inky_Smoke_Tile_Chromatic}
  MaterialFunctionCall(Emissive=Multiply(A=@n1.RGB, B=LinearInterpolate(A=@n13, B=@n17.G) {constAlpha=0.7})) {materialFunction=&MF_FX_EmissiveAdjustments}
  n18 = ScalarParameter() {defaultValue=50, parameterName=DepthFadeOutDist}
  Multiply(A=@n1.A, B=ShadingPathSwitch(Default=DepthFade(Opacity=@n12, FadeDistance=@n18) {fadeDistanceDefault=50}, Forward=@n12, Mobile=@n12))
 ]
}
